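# Base translucent, unlit, two-sided material for glowing eye trail cosmetic effects on skeletal and Niagara meshes.
# Color blends EyeColor with optional texture-driven second color, bright edge, HDR remap and particle tint.
# Opacity multiplies panning smoke by UV length/width gradients and view-angle fade, then applies depth fade.
# Gradient-weighted noise wiggle and local-axis bends combine with positional pushes and camera offset for vertex displacement, alongside tangent-based width scaling.
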Material /BRCosmetics/Effects/Fort_Effects/Materials/EyeTrails/M_Eyetrail_Base_New
graph {
 properties {
  bUsedWithMeshDeformer=true
  bUsedWithNiagaraMeshParticles=true
  bUsedWithSkeletalMesh=true
  blendMode=BLEND_Translucent
  pixelDepthOffsetMode=PDOM_Legacy
  shadingModel=MSM_Unlit
  twoSided=true
  usedShadingModels=Unlit
 }
 nodes [
  n1 = Power(
   Base=OneMinus(None=MaterialFunctionCall() {materialFunction=&LinearGradient}.VGradient)
   Exp=ScalarParameter() {defaultValue=4, group=WPO_Shape, parameterName=Bend_Up_Contrast, sortPriority=1}
  ) {constExponent=4}
  n2 = Transform(None=Constant3Vector() {constant=<rgba 0 0 1 0>}) {transformSourceType=TRANSFORMSOURCE_Local}
  n3 = MaterialFunctionCall() {materialFunction=&LinearGradient}
  n4 = Multiply(
   A=TextureCoordinate() {uTiling=0.02, vTiling=0.5}
   B=ComponentMask(None=VectorParameter() {group=WPO, parameterName=Wiggle UV, sortPriority=4, defaultValue=<rgba 1 1 0 0>}.RGB) {g=true, r=true}
  )
  n5 = VertexColor()
  n6 = Time()
  n7 = ScalarParameter() {defaultValue=0.2, group=WPO, parameterName=Wiggle_Speed, sortPriority=5}
  n8 = TextureSampleParameter2D(UVs=Panner(Coordinate=Add(A=@n4, B=@n5.R), Time=Multiply(A=@n6, B=@n7)) {speedX=0.05, speedY=1}) {
   group=WPO
   parameterName=Wiggle Texture
   samplerSource=SSM_Wrap_WorldGroupSettings
   sortPriority=1
   texture=&T_NoisePack_01
  }
  n9 = Multiply(
   A=Power(Base=OneMinus(None=@n3.VGradient), Exp=ScalarParameter() {defaultValue=1, group=WPO, parameterName=Wiggle_Contrast, sortPriority=1}) {constExponent=4}
   B=Multiply(A=AppendVector(A=@n8.RGB, B=@n8.RGB), B=ScalarParameter() {defaultValue=4, group=WPO, parameterName=WiggleAmount, sortPriority=3})
  )
  n10 = Add(
   A=StaticComponentMaskParameter(None=Transform(None=@n9) {transformSourceType=TRANSFORMSOURCE_Local}) {defaultB=true, defaultG=true, defaultR=true, group=WPO, parameterName=Wiggle Mask, sortPriority=2}
   B=Multiply(A=Multiply(A=@n1, B=@n2), B=ScalarParameter() {group=WPO_Shape, parameterName=Bend_Up})
  )
  n11 = OneMinus(None=MaterialFunctionCall() {materialFunction=&LinearGradient}.VGradient)
  n12 = ScalarParameter() {defaultValue=4, group=WPO_Shape, parameterName=Bend_Foward_Contrast, sortPriority=3}
  n13 = Constant3Vector() {constant=<rgba 0 1 0 0>}
  n14 = Multiply(
   A=Multiply(A=Power(Base=@n11, Exp=@n12) {constExponent=4}, B=Transform(None=@n13) {transformSourceType=TRANSFORMSOURCE_Local})
   B=ScalarParameter() {group=WPO_Shape, parameterName=Bend_Foward, sortPriority=2}
  )
  n15 = Constant3Vector() {constant=<rgba 1 0 0 0>}
  n16 = Constant3Vector() {constant=<rgba 0 1 0 0>}
  n17 = Add(
   A=Multiply(
    A=Transform(None=@n15) {transformSourceType=TRANSFORMSOURCE_Local}
    B=ScalarParameter() {group=WPO_Positioning, parameterName=Push_Mesh_Left/Right}
   )
   B=Multiply(
    A=Transform(None=@n16) {transformSourceType=TRANSFORMSOURCE_Local}
    B=ScalarParameter() {group=WPO_Positioning, parameterName=Push_Mesh_Forward}
   )
  )
  n18 = Multiply(
   A=Transform(None=Constant3Vector() {constant=<rgba 0 0 1 0>}) {transformSourceType=TRANSFORMSOURCE_Local}
   B=ScalarParameter() {group=WPO_Positioning, parameterName=Push_Mesh_Up, sortPriority=1}
  )
  Add(
   A=Add(A=Add(A=@n10, B=@n14), B=Add(A=@n17, B=@n18))
   B=MaterialFunctionCall(Offset Amount=ScalarParameter() {group=WPO, parameterName=Camera Offset}, World Position=WorldPosition().XYZ) {materialFunction=/Engine/Functions/Engine_MaterialFunctions02/WorldPositionOffset/CameraOffset}.Camera Offset
  )
  Constant3Vector() {constant=<rgba 1 0 0 0>}
  Constant() {r=3}
  n19 = ScalarParameter() {defaultValue=0.9, parameterName=Gradient MidPoint}
  Divide(A=ComponentMask() {g=true}, B=@n19)
  LinearInterpolate(B=VertexColor().R, Alpha=ScalarParameter() {defaultValue=1, group=Alpha, parameterName=Use Vertex Color Offset, sortPriority=6})
  n20 = ComponentMask(None=TextureCoordinate()) {g=true}
  LinearInterpolate(Alpha=@n20) {constA=0.5, constB=3}
  LinearInterpolate(
   A=ScalarParameter() {defaultValue=0.25, parameterName=Eyetrail Gradient Exponent End}
   B=ScalarParameter() {defaultValue=1, parameterName=Eyetrail Gradient Exponent Start}
   Alpha=@n20
  )
  n21 = ParticleColor()
  n22 = Multiply(
   A=VectorParameter() {group=Color, parameterName=EyeColor, sortPriority=1, defaultValue=<rgba 10.0282 1.6485 20 1>}.RGB
   B=ScalarParameter() {defaultValue=1, group=Color, parameterName=EyeColor Multiplier}
  )
  n23 = Constant()
  n24 = MaterialFunctionCall(Float2=TextureCoordinate()) {materialFunction=&BreakOutFloat2Components}
  n25 = ScalarParameter() {defaultValue=5, group=Alpha, parameterName=UV Bias, sortPriority=6}
  n26 = Multiply(B=Multiply(A=Power(Base=@n24.G, Exp=@n25), B=ScalarParameter() {defaultValue=1, group=Alpha, parameterName=UV Bias Amount, sortPriority=6})) {constA=0.25}
  n27 = VectorParameter() {group=Alpha, parameterName=TextureTileUV, sortPriority=5, defaultValue=<rgba 1 1 0 0>}
  n28 = Add(A=Multiply(A=TextureCoordinate(), B=ComponentMask(None=@n27.RGB) {g=true, r=true}), B=ParticleRandom())
  n29 = Multiply(A=Time(), B=ScalarParameter() {defaultValue=0.5, group=Alpha, parameterName=Alpha_Speed, sortPriority=6})
  n30 = TextureSampleParameter2D(UVs=Add(
    A=MaterialFunctionCall(X=@n23, Y=@n26) {materialFunction=&MakeFloat2}.Result
    B=Panner(Coordinate=@n28, Time=@n29) {speedX=-0.1, speedY=1}
   )) {
   group=Alpha
   parameterName=Alpha Texture
   samplerSource=SSM_Wrap_WorldGroupSettings
   sortPriority=3
   texture=/Game/Effects/Fort_Effects/Textures/Tile/T_Inky_Smoke_Tile2
  }
  n31 = StaticSwitchParameter(True=LinearInterpolate(A=@n22, B=VectorParameter() {group=Color, parameterName=Color_High, sortPriority=2}.RGB, Alpha=@n30.R), False=@n22) {group=Color, parameterName=Use Gradient Controlled 2 Color System}
  n32 = Multiply(A=ScalarParameter() {defaultValue=10, group=Bright Edge, parameterName=BrightEdge_Brightness, sortPriority=1}, B=@n31)
  n33 = MaterialFunctionCall() {materialFunction=&LinearGradient}
  n34 = ScalarParameter() {defaultValue=10, group=Bright Edge, parameterName=BrightEdgePower, sortPriority=3}
  n35 = Clamp(None=Multiply(
    A=Power(Base=@n33.VGradient, Exp=@n34)
    B=ScalarParameter() {defaultValue=1, group=Bright Edge, parameterName=BrightEdgeMult, sortPriority=2}
   ))
  n36 = StaticSwitchParameter(True=Add(A=Multiply(A=@n32, B=@n35), B=@n31), False=@n31) {group=Bright Edge, parameterName=UseEyeEdgeGlow}
  n37 = StaticSwitchParameter(
   True=MaterialFunctionCall(
    LDR Input=Clamp(None=@n36)
    HDR Tint=@n36
    HDR Intensity Multiplier=ScalarParameter() {defaultValue=1, group=HDR, parameterName=HDR Intensity, sortPriority=1}
   ) {materialFunction=&DeriveHDRfromLDR}.RemappedRange
   False=@n36
  ) {group=HDR, parameterName=UseHDRfromLDR}
  MaterialFunctionCall(Emissive=LinearInterpolate(A=@n37, B=Multiply(A=@n37, B=@n21.RGB), Alpha=ScalarParameter() {group=Color, parameterName=Use Particle Color})) {materialFunction=&MF_FX_EmissiveAdjustments}
  n38 = OneMinus(None=@n20)
  n39 = OneMinus(None=@n19)
  n40 = Multiply(A=Saturate(None=Divide(A=@n20, B=@n19)), B=Saturate(None=Divide(A=@n38, B=@n39)))
  n41 = Power(Base=@n40, Exp=ScalarParameter() {defaultValue=1, parameterName=Eyetrail Gradient Length Exponent})
  n42 = ComponentMask(None=TextureCoordinate()) {r=true}
  n43 = OneMinus(None=@n42)
  n44 = Power(Base=Multiply(A=Multiply(A=@n43, B=@n42)) {constB=4}, Exp=ScalarParameter() {defaultValue=4, parameterName=Eyetrail Gradient Width Exponent})
  n45 = Multiply(A=Saturate(None=Multiply(A=@n41, B=@n44)), B=ScalarParameter() {defaultValue=1, parameterName=Gradient Power})
  n46 = MaterialFunctionCall("FadeSensitivity (optional)"=ScalarParameter() {defaultValue=2, group=Alpha, parameterName=FadeViewAngle, sortPriority=2}) {materialFunction=&FadeBasedOnViewAngle}
  n47 = DepthFade(
   Opacity=Multiply(A=@n30.R, B=Multiply(A=@n45, B=@n46.Result))
   FadeDistance=ScalarParameter() {defaultValue=2, group=Alpha, parameterName=DepthFade, sortPriority=1}
  )
  Multiply(
   A=LinearInterpolate(A=@n47, B=Multiply(A=@n21.A, B=@n47), Alpha=ScalarParameter() {group=Alpha, parameterName=Use Particle Color Alpha})
   B=ScalarParameter() {defaultValue=1, group=Alpha, parameterName=Master_Alpha}
  )
  n48 = ScalarParameter() {group=WPO_Shape, parameterName=Width Base, sortPriority=3}
  n49 = ScalarParameter() {defaultValue=1, group=WPO_Shape, parameterName=Width End, sortPriority=3}
  n50 = MaterialFunctionCall() {materialFunction=&LinearGradient}
  n51 = Power(Base=OneMinus(None=@n50.VGradient), Exp=ScalarParameter() {defaultValue=1, group=WPO_Shape, parameterName=Scale Contrast, sortPriority=3}) {constExponent=4}
  n52 = VertexTangentWS()
  n53 = Abs(None=@n52)
  n54 = Subtract(A=Multiply(A=@n50.UGradient) {constB=2})
  n55 = Multiply(
   A=@n54
   B=MaterialFunctionCall() {materialFunction=&ObjectScale}.Scale XYZ
  )
  Multiply(
   A=Multiply(A=LinearInterpolate(A=@n48, B=@n49, Alpha=@n51), B=Multiply(A=@n53, B=@n55))
   B=ScalarParameter() {group=WPO_Shape, parameterName=ScaleMesh, sortPriority=2}
  )
  Multiply(A=@n54, B=@n52)
  Multiply(A=@n40) {constB=4}
  PerInstanceRandom()
  Power(Base=@n38, Exp=ScalarParameter() {defaultValue=1, parameterName=Bottom Eyetrail Gradient Length})
  Power(Exp=ScalarParameter() {defaultValue=1, parameterName=Top Eyetrail Gradient Length})
  Power(Base=ComponentMask(None=@n43) {r=true}, Exp=ScalarParameter() {defaultValue=1, parameterName=Right Eyetrail Gradient Length})
 ]
}
